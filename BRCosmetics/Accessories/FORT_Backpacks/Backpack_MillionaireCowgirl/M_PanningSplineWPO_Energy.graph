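# Translucent unlit energy material for Niagara mesh-particle spline ribbons on the MillionaireCowgirl backpack.
# SplineThicken builds ribbon width; added WPO combines V-weighted sine/cosine waves with a pull toward a moving-average dynamic parameter, which feeds only X into all three channels.
# Two panning fire-noise samples, shaped by radial gradients and a V falloff, blend two colors.
# Desaturation and time-of-day scaling then feed MF_FX_EmissiveAdjustments.
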Material /BRCosmetics/Accessories/FORT_Backpacks/Backpack_MillionaireCowgirl/M_PanningSplineWPO_Energy
graph {
 properties {
  bUsedWithNiagaraMeshParticles=true
  blendMode=BLEND_Translucent
  opacityMaskClipValue=0.1
  pixelDepthOffsetMode=PDOM_Legacy
  shadingModel=MSM_Unlit
  twoSided=true
  usedShadingModels=Unlit
 }
 nodes [
  n1 = DynamicParameter() {parameterIndex=1, paramNames [Moving Average X, Moving Average Y, Moving Average Z, WPO Intensity]}
  n2 = ComponentMask(None=TextureCoordinate()) {g=true}
  n3 = OneMinus(None=@n2)
  n4 = ScalarParameter() {defaultValue=2.5, parameterName=WPO Frequency}
  n5 = Multiply(A=@n4) {constB=0.5}
  n6 = Time()
  n7 = ScalarParameter() {defaultValue=2, parameterName=WPO Speed}
  n8 = Subtract(
   A=Multiply(A=LinearInterpolate(A=@n4, B=@n5, Alpha=@n2), B=@n3)
   B=Add(A=DynamicParameter() {defaultValue=<rgba 0 0.5 0 1>, paramNames [Time Offset, Random, Flip, Param4]}.Time Offset, B=Multiply(A=@n6, B=@n7))
  )
  n9 = Add(A=ParticleRandom())
  n10 = AppendVector(A=AppendVector(A=Sine(None=@n8) {desc=X}, B=Cosine(None=@n8) {desc=Y}), B=Multiply(A=Sine(None=@n8) {desc=Z, period=3}, B=@n9) {constB=2})
  n11 = ScalarParameter() {defaultValue=0.75, parameterName=WPO Amplitude}
  n12 = Transform(None=Multiply(A=@n3, B=Multiply(A=@n10, B=@n11) {constB=3})) {transformSourceType=TRANSFORMSOURCE_Instance}
  n13 = Multiply(
   A=Multiply(A=@n3, B=@n3)
   B=Subtract(
    A=MaterialFunctionCall(R=@n1.Moving Average X, G=@n1.Moving Average X, B=@n1.Moving Average X) {materialFunction=&AppendMany}.RGB
    B=ParticlePositionWS()
   )
  )
  Add(
   A=MaterialFunctionCall(
    WidthBase=ScalarParameter() {defaultValue=2, parameterName=WidthBase}
    WidthTip=ScalarParameter() {defaultValue=2, parameterName=WidthTip}
    UVs for texturing=TextureCoordinate()
   ) {materialFunction=&SplineThicken}.WorldPositionOffset
   B=Multiply(A=@n1.WPO Intensity, B=Add(A=@n12, B=@n13))
  )
  n14 = Constant() {r=0.9}
  n15 = Constant() {r=1}
  n16 = Constant() {r=0.9}
  n17 = Constant() {r=0.6}
  n18 = Power(Base=ComponentMask(None=TextureCoordinate()) {g=true}) {constExponent=4}
  n19 = Constant2Vector() {g=0.6, r=0.5}
  n20 = Constant() {r=0.6}
  n21 = Constant() {r=4}
  n22 = Multiply(A=Time(), B=@n9)
  n23 = TextureSample(UVs=Panner(Coordinate=TextureCoordinate() {uTiling=0.3, vTiling=0.6}, Time=@n22) {speedX=-0.15, speedY=0.3}) {texture=/Game/Animation/Game/MainPlayer/Emotes/CinderMax/T_Noise_Fire}
  n24 = TextureSample(UVs=Panner(Coordinate=TextureCoordinate() {uTiling=0.3, vTiling=0.3}, Time=@n22) {speedX=0.05, speedY=0.2}) {texture=&T_Noise_Fire}
  n25 = Multiply(
   A=LinearInterpolate(A=@n18, Alpha=DynamicParameter() {defaultValue=<rgba 0 0.5 0 1>, paramNames [Time Offset, Random, Flip, Param4]}.Random) {constA=1, constAlpha=0}
   B=Multiply(
    A=MaterialFunctionCall(CenterPosition=@n19, Radius=@n20, Density=@n21) {materialFunction=&RadialGradientExponential}.RadialGradientExponential
    B=Multiply(A=@n23.G, B=@n24.G)
   )
  )
  n26 = MaterialFunctionCall(CenterPosition=Constant2Vector() {g=0.55, r=0.5}, Radius=Constant() {r=0.6}, Density=Constant() {r=0.6}) {materialFunction=&RadialGradientExponential}
  n27 = Add(A=Power(Base=Divide(A=@n25) {constB=0.05}) {constExponent=4}, B=@n26.RadialGradientExponential)
  n28 = LinearInterpolate(
   A=VectorParameter() {parameterName=Color, defaultValue=<rgba 0.5 0 0.0833 4>}.RGB
   B=VectorParameter() {parameterName=Color B, defaultValue=<rgba 5 0 3.3333 0>}.RGB
   Alpha=Multiply(A=@n18, B=Saturate(None=@n27))
  )
  MaterialFunctionCall(Emissive=Multiply(
    A=MaterialFunctionCall(Value_Morning=@n14, Value_Day=@n15, Value_Evening=@n16, Value_Night=@n17) {materialFunction=&Scalar_Time_of_Day}.Result
    B=Desaturation(None=@n28, Fraction=@n26.RadialGradientExponential)
   )) {materialFunction=&MF_FX_EmissiveAdjustments}
 ]
}
