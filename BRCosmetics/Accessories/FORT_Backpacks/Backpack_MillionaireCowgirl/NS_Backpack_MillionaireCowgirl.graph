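# Niagara effect for the MillionaireCowgirl back bling that renders a pair of glowing eyes with smoke.
# Partitioning mirrors sphere meshes, rising sparkle sprites and pulsing flare sprites with lights across both eyes, with flare and light colors set by time of day.
# Smoke meshes lag behind movement: the emitter smooths a world-space target, constrains it by distance and local axes, then feeds position and intensity to dynamic material parameters for world position offset.
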NiagaraSystem /BRCosmetics/Accessories/FORT_Backpacks/Backpack_MillionaireCowgirl/NS_Backpack_MillionaireCowgirl
dependencies {
 usedDataInterfaces [&NiagaraDataInterfaceCurve, &NiagaraDataInterfaceArrayFloat3, &NiagaraDataInterfaceArrayFloat]
 usedDynamicInputs []
 usedModules [&SystemState]
}
emitters {
 EyeBits {
  emitterUpdateScript [
   &EmitterState(Life Cycle Mode=::NewEnumerator1, Loop Behavior=::NewEnumerator1, Loop Duration Mode=::NewEnumerator1, UseLoopDelay=false)
   &SpawnRate(SpawnRate=64, Use Spawn Probability=false)
  ]
  particleSpawnScript [
   &PartitionParticles()
   &InitializeParticle(
    Lifetime Max=0.6
    Lifetime Min=0.3
    Lifetime Mode=::NewEnumerator1
    Position Offset=&Lerp_Vector(
     A=<5.75 -5.5 -3>
     B=<-5.75 -5.5 -3>
     Alpha=&MakeFloatFromInt(Int=$Particles.PartitionParticles.Partition)
    )
    Sprite Size Mode=::NewEnumerator1
    Uniform Sprite Size Max=0.6
    Uniform Sprite Size Min=0.5
    UsePositionOffset=true
    Write Material Random=true
   )
   &ShapeLocation(MaskSpawnGroup=false, Override Seed=false, Sphere Radius=0.5, UseOwnerScale=true)
   &AddVelocity(
    Module.Use Falloff Distance=false
    Origin Offset=<0 -5 -3>
    Override Seed=false
    Velocity Mode=::NewEnumerator1
    Velocity Speed=&RandomRangeFloat(Minimum=2, Maximum=4)
   )
   &ApplyInitialForces()
  ]
  particleUpdateScript [
   &ParticleState()
   &AccelerationForce(Acceleration=<0 0 10>)
   &SolveForcesAndVelocity(Clamp Velocity=false, Limit Acceleration=false)
   &ScaleSpriteSize(Uniform Curve Sprite Scale=DataInterface() {propertyValues {Curve {keys [{tangentMode=RCTM_User, time=0, value=0}, <at 0.05 1 RCIM_Cubic>, {arriveTangent=-3.1733, tangentMode=RCTM_User, time=1, value=0}]}}}) {module=ScaleSpriteSize001}
  ]
  rendererClasses [&NiagaraSpriteRendererProperties]
  renderers [{rendererClass=&NiagaraSpriteRendererProperties}]
 }
 EyeFlares {
  emitterUpdateScript [
   &EmitterState(Life Cycle Mode=::NewEnumerator1, Loop Behavior=::NewEnumerator1, Loop Duration Mode=::NewEnumerator1, UseLoopDelay=false)
   &SpawnBurst_Instantaneous(Spawn Count=2, Use Spawn Probability=false, UseLoopCountLimit=false)
  ]
  particleSpawnScript [
   &PartitionParticles()
   &InitializeParticle(
    Color=&FN_Select_Color_from_Time_Of_Day(Morning=<rgba 0.8 0 0.2667 0.6>, Day=<rgba 1 0 0.3333 0.6>, Evening=<rgba 0.8 0 0.2667 0.6>, Night=<rgba 0.3 0 0.25 0.6>)
    Mesh Scale Mode=::NewEnumerator3
    Position Offset=&Lerp_Vector(A=<6.5 -5 -3>, B=<-6.5 -5 -3>, Alpha=&MakeFloatFromInt(Int=$Particles.PartitionParticles.Partition))
    Sprite Size Mode=::NewEnumerator3
    Uniform Sprite Size=16
    UsePositionOffset=true
    Write Material Random=true
   )
   /SetVariables(Particles.LightColor=&FN_Select_Color_from_Time_Of_Day(Morning=<rgba 400 0 99.9998 1>, Day=<rgba 500 0 83.3333 1>, Evening=<rgba 400 0 99.9998 1>, Night=<rgba 150 0 124.9998 1>)) {isSetParameters=true}
   &CameraOffset(Camera Offset Amount=6)
  ]
  particleUpdateScript [
   &ParticleState(Kill Particles When Lifetime Has Elapsed=false, Let Infinitely Lived Particles Die When Emitter Deactivates=true)
   &ScaleSpriteSize(
    Scale Sprite Size Mode=::NewEnumerator0
    Uniform Scale Factor=&Waveform("Waveform [1]"=::NewEnumerator0, "[1] Frequency"=0.5, "[1] Amplitude Min/Max"=<0.75 1>)
   )
   /SetVariables(Particles.LightRadius=&Waveform("Waveform [1]"=::NewEnumerator0, "[1] Frequency"=0.5, "[1] Amplitude Min/Max"=<50 100>)) {isSetParameters=true}
  ]
  rendererClasses [&NiagaraSpriteRendererProperties, &NiagaraLightRendererProperties]
  renderers [
   {rendererClass=&NiagaraSpriteRendererProperties}
   {rendererClass=&NiagaraLightRendererProperties, rendererIndex=1}
  ]
 }
 EyeSmoke001 {
  emitterSpawnScript [/SetVariables(
    Emitter.InterpolateDistanceLimit=20
    Emitter.InterpolateRate=20
    Emitter.InterpolateTarget:NiagaraPosition=&TransformPosition(Position=<0 -10 0>, Source Space=::Local, Destination Space=::World)
   ) {isSetParameters=true}]
  emitterUpdateScript [
   &EmitterState(Life Cycle Mode=::NewEnumerator1, Loop Behavior=::NewEnumerator1, Loop Duration Mode=::NewEnumerator1, UseLoopDelay=false)
   &InterpolateOverTime(
    Initial Vector Value:NiagaraPosition=$Emitter.InterpolateTarget
    Rate Of Change:NiagaraFloat=$Emitter.InterpolateRate
    Target Vector Value:NiagaraPosition=$Emitter.InterpolateTarget
    Variable Type=::NewEnumerator2
   )
   /SetVariables(Emitter.InterpolateTarget:NiagaraPosition=&TransformPosition(Position=<0 -10 -5>, Source Space=::Local, Destination Space=::World)) {isSetParameters=true}
   /SetVariables(Emitter.InterpolateOverTime.InterpolateOverTime_V3_.Moving Average:Vector3f=&AddVectorToPosition(
     Position=Linked() {linkedVariable {name=StackContext.InterpolateTarget, type=Vector3f}}
     Vector=&Multiply_VectorByFloat(
      Vector=&NormalizeVector(
       Vector To Normalize=&Subtract_Vector(A=$StackContext.InterpolateOverTime.InterpolateOverTime_V3_.Moving Average, B=$StackContext.InterpolateTarget)
       Safe Normalize=true
       Safe Fallback Threshold=0
       Safe Fallback Vector=<0 0 0.001>
      )
      Float=/Niagara/DynamicInputs/Math/Min_Float(
       A:NiagaraFloat=&VectorLength(Vector=&Subtract_Vector(A=$StackContext.InterpolateOverTime.InterpolateOverTime_V3_.Moving Average, B=$StackContext.InterpolateTarget))
       B:NiagaraFloat=$StackContext.InterpolateDistanceLimit
      )
     )
    )) {isSetParameters=true}
   /SetVariables(Emitter.InterpolateOverTime.InterpolateOverTime_V3_.Moving Average:Vector3f=&TransformPosition(
     Position=&MakeVector(
      X=&MakeFloatFromVector(
       Vector=&TransformPosition(Position=$Emitter.InterpolateOverTime.InterpolateOverTime_V3_.Moving Average, Source Space=::World, Destination Space=::Local)
       Channel=::NewEnumerator0
      )
      Y=&Min_Float(
       A:NiagaraFloat=&MakeFloatFromVector(
        Vector=&TransformPosition(Position=$Emitter.InterpolateOverTime.InterpolateOverTime_V3_.Moving Average, Source Space=::World, Destination Space=::Local)
        Channel=::NewEnumerator1
       )
       B:NiagaraFloat=&MakeFloatFromVector(
        Vector=&TransformPosition(Position=$Emitter.InterpolateTarget, Source Space=::World, Destination Space=::Local)
        Channel=::NewEnumerator1
       )
      )
      Z=&Lerp_Float(
       A=&MakeFloatFromVector(
        Vector=&TransformPosition(Position=$Emitter.InterpolateOverTime.InterpolateOverTime_V3_.Moving Average, Source Space=::World, Destination Space=::Local)
        Channel=::NewEnumerator2
       )
       B=&MakeFloatFromVector(
        Vector=&TransformPosition(Position=$Emitter.InterpolateTarget, Source Space=::World, Destination Space=::Local)
        Channel=::NewEnumerator2
       )
       Alpha=0.5
      )
     )
     Source Space=::Local
     Destination Space=::World
    )) {isSetParameters=true}
   /SetVariables(Emitter.WPOIntensityModulate:NiagaraFloat=&Divide_Float(
     A=&DistanceBetweenPositions(
      A=Linked() {linkedVariable {name=StackContext.InterpolateTarget, type=Vector3f}}
      B=$StackContext.InterpolateOverTime.InterpolateOverTime_V3_.Moving Average
     )
     B=$Emitter.InterpolateDistanceLimit
    )) {isSetParameters=true}
   &SpawnBurst_Instantaneous(Spawn Count=10, Use Spawn Probability=false, UseLoopCountLimit=false)
  ]
  particleSpawnScript [
   &PartitionParticles()
   &InitializeParticle(
    Color Mode=::NewEnumerator0
    Mesh Scale Mode=::NewEnumerator3
    Position Mode=::NewEnumerator1
    Position Offset=<0 0.5 0>
    UsePositionOffset=true
    Write Material Random=true
   )
   /SetVariables(Particles.BiasedRandom=&FloatFromCurve(
     FloatCurve=DataInterface() {propertyValues {Curve {keys [
         <at 0 0>
         <at 0.8 0.2>
         {arriveTangent=6.4981, arriveTangentWeight=0.2667, interpMode=RCIM_Cubic, leaveTangent=6.4981, tangentMode=RCTM_User, time=1, value=1}
        ]}}}
     CurveIndex=&RandomRangeFloat(Minimum=0, Maximum=1)
    )) {isSetParameters=true}
   /SetVariables(Particles.MeshIndex=&RandomRangeInt(Minimum=0, Maximum=4)) {isSetParameters=true}
   /SetVariables(Particles.Scale=&SelectVectorFromArray(
     Direct Array Index=$Particles.PartitionParticles.Partition
     Vector Selection Array=DataInterface() {propertyValues {FloatData [<-1 1 1>, <1 1 1>]}}
    )) {isSetParameters=true}
   &DynamicMaterialParameters(
    Index 0 Param 1=&RandomRangeFloat(Minimum=0, Maximum=999)
    Index 0 Param 2=&Lerp_Float(A=0, B=1, Alpha=$Particles.BiasedRandom)
    Index 0 Param 3=&SelectFloatFromArray(Direct Array Index=$Particles.PartitionParticles.Partition, Float Selection Array=DataInterface() {propertyValues {FloatData [-1,1]}})
    Module.Index0Param0Write=true
    Module.Index0Param1Write=true
    Module.Index0Param2Write=false
    Module.Index0Param3Write=false
   )
  ]
  particleUpdateScript [
   &ParticleState(Kill Particles When Lifetime Has Elapsed=false, Let Infinitely Lived Particles Die When Emitter Deactivates=true)
   &DynamicMaterialParameters(
    Index 1 Param 1=&MakeFloatFromVector(
     Vector=Linked() {linkedVariable {name=Emitter.InterpolateOverTime.InterpolateOverTime_V3_.Moving Average, type=NiagaraPosition}}
     Channel=::NewEnumerator0
    )
    Index 1 Param 2=&MakeFloatFromVector(
     Vector=Linked() {linkedVariable {name=Emitter.InterpolateOverTime.InterpolateOverTime_V3_.Moving Average, type=NiagaraPosition}}
     Channel=::NewEnumerator1
    )
    Index 1 Param 3=&MakeFloatFromVector(
     Vector=Linked() {linkedVariable {name=Emitter.InterpolateOverTime.InterpolateOverTime_V3_.Moving Average, type=NiagaraPosition}}
     Channel=::NewEnumerator2
    )
    Index 1 Param 4=$Emitter.WPOIntensityModulate
    Module.Index1Param0Write=true
    Module.Index1Param1Write=true
    Module.Index1Param2Write=true
    Module.Index1Param3Write=true
    Write Parameter Index 0=false
    Write Parameter Index 1=true
   ) {module=DynamicMaterialParameters001}
   /SetVariables(Particles.DebugPosition:NiagaraPosition=&TransformPosition(Position=$Emitter.InterpolateOverTime.InterpolateOverTime_V3_.Moving Average, Source Space=::World, Destination Space=::Local)) {enabled=false, isSetParameters=true}
  ]
  rendererClasses [&NiagaraMeshRendererProperties]
  renderers [
   {rendererClass=&NiagaraMeshRendererProperties, rendererIndex=0}
   {rendererClass=&NiagaraMeshRendererProperties, rendererIndex=1}
  ]
 }
 EyeSpheres {
  emitterUpdateScript [
   &EmitterState(Life Cycle Mode=::NewEnumerator1, Loop Behavior=::NewEnumerator1, Loop Duration Mode=::NewEnumerator1, UseLoopDelay=false)
   &SpawnBurst_Instantaneous(Spawn Count=2, Use Spawn Probability=false, UseLoopCountLimit=false)
  ]
  particleSpawnScript [
   &PartitionParticles()
   &InitializeParticle(
    Color=<rgba 2 2 2 1>
    Mesh Scale Mode=::NewEnumerator3
    Position Offset=&Lerp_Vector(
     A=<5.75 -5.5 -3.25>
     B=<-5.75 -5.5 -3.25>
     Alpha=&MakeFloatFromInt(Int=$Particles.PartitionParticles.Partition)
    )
    UsePositionOffset=true
    Write Material Random=true
   )
  ]
  particleUpdateScript [&ParticleState(Kill Particles When Lifetime Has Elapsed=false, Let Infinitely Lived Particles Die When Emitter Deactivates=true)]
  rendererClasses [&NiagaraMeshRendererProperties]
  renderers [{rendererClass=&NiagaraMeshRendererProperties, rendererIndex=0}]
 }
}
summary {
 systemName=NS_Backpack_MillionaireCowgirl
 emitters [
  {emitterName=EyeFlares, rendererClasses [&NiagaraSpriteRendererProperties, &NiagaraLightRendererProperties]}
  {emitterName=EyeSpheres, rendererClasses [&NiagaraMeshRendererProperties]}
  {emitterName=EyeBits, rendererClasses [&NiagaraSpriteRendererProperties]}
  {emitterName=EyeSmoke001, rendererClasses [&NiagaraMeshRendererProperties]}
 ]
 userVariables [
  User.bIsFrontEnd {type=NiagaraBool}
  User.bIsFrontEndPreview {type=NiagaraBool}
 ]
}
